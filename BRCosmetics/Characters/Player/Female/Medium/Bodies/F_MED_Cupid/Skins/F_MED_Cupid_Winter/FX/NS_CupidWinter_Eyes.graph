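# Niagara system that puts glowing, pulsing eye lights on the winter skin of the female medium Cupid character.
# A single emitter bursts two particles, pinned via spawn groups to the head's L_eye and R_eye bones, drawn by sprite and light renderers.
# Repeating dip curves pulse alpha, size and light color alpha, with sprite alpha also scaled by time of day.
# User flags adjust brightness and size for the front end and set VisibilityTag when the head is hidden.
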NiagaraSystem /BRCosmetics/Characters/Player/Female/Medium/Bodies/F_MED_Cupid/Skins/F_MED_Cupid_Winter/FX/NS_CupidWinter_Eyes
dependencies {
 usedDataInterfaces [&NiagaraDataInterfaceCurve, &NiagaraDataInterfaceSkeletalMesh, &NiagaraDataInterfaceArrayFloat3]
 usedDynamicInputs []
 usedModules [&SystemState]
}
emitters {eyes_light {
  emitterUpdateScript [
   &EmitterState(
    Inactive Response=::NewEnumerator1
    Life Cycle Mode=::NewEnumerator1
    Loop Behavior=::NewEnumerator1
    Loop Duration Mode=::NewEnumerator1
    Reset Age On Awaken=true
    Scalability Mode=::NewEnumerator1
    UseLoopDelay=false
   )
   &SpawnBurst_Instantaneous(Spawn Count=2, Use Spawn Probability=false)
  ]
  particleSpawnScript [
   /Niagara/Modules/Spawn/Initialization/InitializeParticle(
    Color=<rgba 1.9341 1.9798 2 1.6>
    Mass=&UniformRangedFloat(Minimum=0.6, Maximum=1.2)
    Module.Write Color=true
    Module.Write Lifetime=true
    Module.Write Mass=false
    Module.Write Material Random=true
    Module.Write Position=false
    Module.Write Scale=false
    Module.Write SpriteRotation=false
    Module.Write SpriteSize=true
    Position=&SimulationPosition()
    Sprite Size=&Multiply_Vector2(
     A=<15 15>
     B=&Vector2DFromFloat(Value=&MakeCustomFloatFromBool(Bool=$User.Is_it_in_FrontEnd, True Float=0.8, False Float=0.9))
    )
   )
   /SetVariables(Particles.Light_Size=5) {isSetParameters=true}
   &FN_SetSpawnGroup_v2(Number of Spawn Groups=2)
  ]
  particleUpdateScript [
   &ParticleState(Kill Particles When Lifetime Has Elapsed=false, Let Infinitely Lived Particles Die When Emitter Deactivates=true)
   &ScaleColor(
    Scale Alpha=&Multiply_Float(
     A=&Multiply_Float(
      A=&MakeCustomFloatFromBool(Bool=$User.Is_it_in_FrontEnd, True Float=2.7, False Float=1)
      B=&FNTimeOfDayScalar(Night Scale=0.8, Morning Scale=0.3, Evening Scale=0.5)
     )
     B=&FloatFromCurve(
      FloatCurve=DataInterface() {propertyValues {Curve {keys [<at 0 1>,<at 0.0539 1>,<at 0.1557 0.8>,<at 0.2614 1>,<at 0.3742 0.8>,<at 0.456 1>,<at 0.5872 0.8>,<at 0.7535 1>,<at 0.8976 0.8>,<at 1 1>]}}}
      CurveIndex=$System.NormalizedLoopAge
     )
    )
    Scale RGB=<1.3 1.3 1.3>
    ScaleA=true
    ScaleRGB=true
   )
   &SolveForcesAndVelocity(Clamp Velocity=false, Limit Acceleration=false)
   /SetVariables(Particles.Light_Color:LinearColor=&MakeLinearColorFromVectorAndFloat(
     "Vector (RGB)"=&MakeVectorFromLinearColorRGB(Color=<rgba 11.6875 12 11.375 1.5>)
     "Float (Alpha)"=&FloatFromCurve(FloatCurve=DataInterface() {propertyValues {Curve {keys [<at 0 1>,<at 0.0539 1>,<at 0.1557 0.5>,<at 0.2614 1>,<at 0.3742 0.5>,<at 0.456 1>,<at 0.5872 0.5>,<at 0.7535 1>,<at 0.8976 0.5>,<at 1 1>]}}})
    )) {isSetParameters=true}
   &CameraOffset(Camera Offset Amount=1.5)
   &SkeletalMeshLocation(
    Apply Sampled Orientation to Mesh Particles=true
    Bone / Socket Index=$Particles.SpawnGroup
    Bone Sampling Mode=::NewEnumerator4
    Inherit Velocity=false
    Interpolate Mesh Sampling=false
    MaskSpawnGroup=false
    Sampled Position Offset=&SelectVectorFromArray(Direct Array Index=$Particles.SpawnGroup, Vector Selection Array=DataInterface() {propertyValues {FloatData [<0 0 0>, <0 0 0>]}})
    Sampled Position Offset Coordinate Space:ENiagaraMeshTransforms=::NewEnumerator3
    Skeletal Mesh=DataInterface() {propertyValues {
      DefaultMesh=null
      ExcludeBoneName=root
      OptionalComponentClass=null
      PreviewMesh=/BRCosmetics/Characters/Player/Female/Medium/Heads/F_MED_Cupid_Head/Meshes/Female_Medium_Cupid_Head
      SkinningMode=SkinOnTheFly
      SoftSourceActor=null
      SourceMode=Default
      UvSetIndex=0
      WholeMeshLOD=-1
      bExcludeBone=false
      bReadDeformedGeometry=true
      bRequireCurrentFrameData=true
      ComponentTags []
      FilteredBones [L_eye,R_eye]
      FilteredSockets []
      MeshUserParameter {parameter {name=None, typeDefHandle {registeredTypeIndex=1}, varData []}}
      SamplingRegions []
     }}
   )
   &DynamicMaterialParameters(
    Index 0 Param 1=1
    Index 0 Param 2=1
    Index 0 Param 3=1
    Index 0 Param 4=1
    Index 1 Param 1=5
    Index 1 Param 2=1
    Index 1 Param 3=1
    Index 1 Param 4=1
    Module.Param0WriteEnabled=true
    Module.Param0WriteEnabled1=true
    Module.Param1WriteEnabled=true
    Module.Param1WriteEnabled1=true
    Module.Param2WriteEnabled=true
    Module.Param2WriteEnabled1=true
    Module.Param3WriteEnabled=true
    Module.Param3WriteEnabled1=true
    Write Parameter Index 1=true
   )
   &ScaleSpriteSize(Scale Factor=&Vector2DFromFloat(Value=&FloatFromCurve(
      FloatCurve=DataInterface() {propertyValues {Curve {keys [<at 0 1>,<at 0.0539 1>,<at 0.1557 0.985>,<at 0.2614 1>,<at 0.3742 0.985>,<at 0.456 1>,<at 0.5872 0.985>,<at 0.7535 1>,<at 0.8976 0.985>,<at 1 1>]}}}
      CurveIndex=$System.NormalizedLoopAge
     )))
   /SetVariables(Particles.VisibilityTag=&MakeCustomIntFromBool(Bool=$User.bisHeadHidden, True INT=1, False INT=0)) {isSetParameters=true}
  ]
  rendererClasses [&NiagaraLightRendererProperties, &NiagaraSpriteRendererProperties]
  renderers [
   {rendererClass=&NiagaraLightRendererProperties, rendererIndex=0}
   {rendererClass=&NiagaraSpriteRendererProperties, rendererIndex=1}
  ]
 }}
summary {
 systemName=NS_CupidWinter_Eyes
 emitters [{emitterName=eyes_light, rendererClasses [&NiagaraLightRendererProperties, &NiagaraSpriteRendererProperties]}]
 userVariables [
  User.bisHeadHidden {type=NiagaraBool}
  User.Is_it_in_FrontEnd {type=NiagaraBool}
 ]
}
